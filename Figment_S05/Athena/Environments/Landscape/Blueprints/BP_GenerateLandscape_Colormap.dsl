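; Editor utility Blueprint that generates a landscape colormap texture from a scene capture.
; Construction script pushes display toggles and colormap/grass-edge tuning values into MPC_Landscape to drive landscape shading.
; Separate custom events trigger the SceneCapture2D and overwrite the target Texture2D from the render target via an editor-only conversion.
; Restore Landscape Mats assigns Render Material to all streaming proxies and the first Landscape actor.
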(event Custom|Restore%20Landscape%20Mats
  (bind _outactors (Actor|GetAllActorsOfClass "/Script/Landscape.LandscapeStreamingProxy"))
  (for _array_element _outactors
    (Variables|/Script/Landscape.LandscapeProxy|SetLandscapeMaterial (Variables|Default|GetRender%20Material) _array_element))
  (bind _outactors_2 (Actor|GetAllActorsOfClass "/Script/Landscape.Landscape"))
  (Variables|/Script/Landscape.LandscapeProxy|SetLandscapeMaterial (Variables|Default|GetRender%20Material) (Utilities|Array|Get(acopy) _outactors_2)))

(event Custom|Trigger%20Scene%20Capture
  (Rendering|SceneCapture|CaptureScene (Variables|/Script/Engine.SceneCapture2D|GetCaptureComponent2D (Variables|Default|GetScene%20Capture%20Actor)) :!class /Script/Engine.SceneCaptureComponent2D))

(event Custom|Overwrite%20Texture%20Target
  (EditorScripting|Rendering|ConvertRenderTargettoTexture2DEditorOnly (Variables|Default|GetRender%20Target) (Variables|Default|GetTarget%20Texture2D)))

(fn ConstructionScript
  (Rendering|Material|SetScalarParameterValue "/Game/Athena/Environments/Landscape/MPC/MPC_Landscape.MPC_Landscape" "DisplayLowDetailShader" (Math|Conversions|ToFloat(Boolean) (Variables|Default|GetDisplay%20Low%20Detail%20Shader)) :!class /Script/Engine.KismetMaterialLibrary)
  (Rendering|Material|SetScalarParameterValue "/Game/Athena/Environments/Landscape/MPC/MPC_Landscape.MPC_Landscape" "RenderColorMap" (Math|Conversions|ToFloat(Boolean) (Variables|Default|GetDisplay%20Color%20Map)) :!class /Script/Engine.KismetMaterialLibrary)
  (Rendering|Material|SetScalarParameterValue "/Game/Athena/Environments/Landscape/MPC/MPC_Landscape.MPC_Landscape" "DisplayBiomeRGB" (Math|Conversions|ToFloat(Boolean) (Variables|Default|GetDisplay%20Biome%20RGB)) :!class /Script/Engine.KismetMaterialLibrary)
  (Rendering|Material|SetScalarParameterValue "/Game/Athena/Environments/Landscape/MPC/MPC_Landscape.MPC_Landscape" "ColormapDivisor" (Variables|Default|GetDivisor) :!class /Script/Engine.KismetMaterialLibrary)
  (Rendering|Material|SetScalarParameterValue "/Game/Athena/Environments/Landscape/MPC/MPC_Landscape.MPC_Landscape" "ColormapTransitionBias" (Variables|Default|GetGrass%20Cliff%20Bias) :!class /Script/Engine.KismetMaterialLibrary)
  (Rendering|Material|SetScalarParameterValue "/Game/Athena/Environments/Landscape/MPC/MPC_Landscape.MPC_Landscape" "GrassEdgeZ" (Variables|Default|GetGrass%20Edge%20Z) :!class /Script/Engine.KismetMaterialLibrary)
  (Rendering|Material|SetScalarParameterValue "/Game/Athena/Environments/Landscape/MPC/MPC_Landscape.MPC_Landscape" "GrassEdgeDesatWidth" (Variables|Default|GetGrass%20Edge%20Desat%20Width) :!class /Script/Engine.KismetMaterialLibrary))
